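# Physics asset for the female medium Grandeur body cosmetic, set up for secondary motion on its zippers, hood, coat flaps and scarf lining.
# Capsules on pelvis, spine, neck, clavicles and upper arms form a torso chain joined by angularly free constraints.
# Small MSTR anchors on spine_04 and spine_05 carry coat, zipper and hood bodies with tight twist limits.
# Scarf lining boxes hang directly from neck_01 with narrow swing.
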PhysicsAsset /BRCosmetics/Characters/Player/Female/Medium/Bodies/F_MED_Grandeur/Meshes/F_MED_Grandeur_Physics
bodies {
 clavicle_l [{shapeType=Capsule, length=9.6738, radius=8.9645, shapeName=clavicle_l_capsule, center=<6.6757 -3.4166 -0.3662>, rotation=<rot 90 0 0>}]
 clavicle_r [{shapeType=Capsule, length=9.8822, radius=9.6487, shapeName=clavicle_r_capsule, center=<-5.9688 2.5339 0.8579>, rotation=<rot 90 0 0>}]
 dyn_body_zipper_l_01 [{shapeType=Capsule, length=1.2738, radius=0.656, shapeName=dyn_body_zipper_l_01_capsule, center=<2 0 0>, rotation=<rot 90 0 0>}]
 dyn_body_zipper_l_MSTR [{shapeType=Capsule, length=0.505, radius=0.505, shapeName=dyn_body_zipper_l_MSTR_capsule, rotation=<rot 0 -90 0>}]
 dyn_body_zipper_r_01 [{shapeType=Capsule, length=1.2742, radius=0.6568, shapeName=dyn_body_zipper_r_01_capsule, center=<2 0 0>, rotation=<rot 90 0 0>}]
 dyn_body_zipper_r_MSTR [{shapeType=Capsule, length=0.505, radius=0.505, shapeName=dyn_body_zipper_r_MSTR_capsule, rotation=<rot 0 -90 0>}]
 dyn_coat_bk_l_01 [{shapeType=Capsule, length=6.7748, radius=2.25, shapeName=dyn_coat_bk_l_01_capsule, center=<3 0 0>, rotation=<rot 0 -6.8902 -51.9166>}]
 dyn_coat_bk_l_MSTR [{shapeType=Capsule, length=0.505, radius=0.505, shapeName=dyn_coat_bk_l_MSTR_capsule, rotation=<rot 0 -90 0>}]
 dyn_coat_bk_r_01 [{shapeType=Capsule, length=6.7748, radius=2.25, shapeName=dyn_coat_bk_r_01_capsule, center=<3 0 0>, rotation=<rot 0 6.8902 38.0834>}]
 dyn_coat_bk_r_MSTR [{shapeType=Capsule, length=0.505, radius=0.505, shapeName=dyn_coat_bk_r_MSTR_capsule, rotation=<rot 0 -90 0>}]
 dyn_coat_fr_l_01 [{shapeType=Capsule, length=7.3039, radius=2, shapeName=dyn_coat_fr_l_01_capsule, center=<3 0 0>, rotation=<rot 45 -90 0>}]
 dyn_coat_fr_l_MSTR [{shapeType=Capsule, length=0.505, radius=0.505, shapeName=dyn_coat_fr_l_MSTR_capsule, rotation=<rot 0 -90 0>}]
 dyn_coat_fr_r_01 [{shapeType=Capsule, length=7.3039, radius=2, shapeName=dyn_coat_fr_r_01_capsule, center=<3 0 0>, rotation=<rot -45 -90 0>}]
 dyn_coat_fr_r_MSTR [{shapeType=Capsule, length=0.505, radius=0.505, shapeName=dyn_coat_fr_r_MSTR_capsule, rotation=<rot 0 -90 0>}]
 dyn_hood_01 [{shapeType=Capsule, length=8.4128, radius=6.0915, shapeName=dyn_hood_01_capsule, center=<5.6716 0 0>, rotation=<rot 0 0 90>}]
 dyn_hood_MSTR [{shapeType=Capsule, length=0.505, radius=0.505, shapeName=dyn_hood_MSTR_capsule, rotation=<rot 0 -90 0>}]
 dyn_scarf_lining_l_MSTR [{
   shapeType=Box
   extentX=5.2693
   extentY=1.2509
   extentZ=6.8694
   shapeName=dyn_scarf_lining_l_MSTR_box
   center=<-1.8722 3.3937 -1.5835>
   rotation=<rot 7.3979 -90 40.5302>
  }]
 dyn_scarf_lining_r_MSTR [{
   shapeType=Box
   extentX=5.1962
   extentY=1.4201
   extentZ=5.1638
   shapeName=dyn_scarf_lining_r_MSTR_box
   center=<1.2882 -3.9259 0.9594>
   rotation=<rot 13.1827 -90 42.9334>
  }]
 neck_01 [{shapeType=Capsule, length=0.505, radius=0.505, shapeName=neck_01_capsule, rotation=<rot 0 -90 0>}]
 pelvis [{shapeType=Capsule, length=15.8173, radius=14.05, shapeName=pelvis_capsule, center=<-1.6347 -0.7478 0.1598>}]
 spine_01 [{shapeType=Capsule, length=10.3458, radius=8.5339, shapeName=spine_01_capsule, center=<9.8927 1.5187 -0.0026>, rotation=<rot 90 0 0>}]
 spine_02 [{shapeType=Capsule, length=5.1095, radius=8.7503, shapeName=spine_02_capsule, center=<3.0251 1.9328 0>, rotation=<rot 90 0 0>}]
 spine_03 [{shapeType=Capsule, length=8.3852, radius=2.2648, shapeName=spine_03_capsule, center=<-1.2949 -6.9756 0>, rotation=<rot 0 -90 0>}]
 spine_04 [{shapeType=Capsule, length=1.811, radius=11.483, shapeName=spine_04_capsule, center=<-1.0577 2.7594 0.0112>}]
 spine_05 [{shapeType=Capsule, length=13.048, radius=11.4167, shapeName=spine_05_capsule, center=<5.6463 1.884 0.582>}]
 upperarm_twist_01_l [{shapeType=Capsule, length=10.8743, radius=6.7892, shapeName=upperarm_twist_01_l_capsule, rotation=<rot 90 0 0>}]
 upperarm_twist_01_r [{shapeType=Capsule, length=11.0121, radius=6.5744, shapeName=upperarm_twist_01_r_capsule, rotation=<rot 90 0 0>}]
}
constraints [
 {
  bone1Name=spine_01
  bone2Name=pelvis
  swing1LimitDegrees=45
  swing1Motion=Free
  swing2LimitDegrees=45
  swing2Motion=Free
  twistLimitDegrees=45
  twistMotion=Free
 }
 {
  bone1Name=spine_02
  bone2Name=spine_01
  swing1LimitDegrees=45
  swing1Motion=Free
  swing2LimitDegrees=45
  swing2Motion=Free
  twistLimitDegrees=45
  twistMotion=Free
 }
 {
  bone1Name=spine_03
  bone2Name=spine_02
  swing1LimitDegrees=45
  swing1Motion=Free
  swing2LimitDegrees=45
  swing2Motion=Free
  twistLimitDegrees=45
  twistMotion=Free
 }
 {
  bone1Name=spine_04
  bone2Name=spine_03
  swing1LimitDegrees=45
  swing1Motion=Free
  swing2LimitDegrees=45
  swing2Motion=Free
  twistLimitDegrees=45
  twistMotion=Free
 }
 {
  bone1Name=spine_05
  bone2Name=spine_04
  swing1LimitDegrees=45
  swing1Motion=Free
  swing2LimitDegrees=45
  swing2Motion=Free
  twistLimitDegrees=45
  twistMotion=Free
 }
 {
  bone1Name=clavicle_l
  bone2Name=spine_05
  swing1LimitDegrees=45
  swing1Motion=Free
  swing2LimitDegrees=45
  swing2Motion=Free
  twistLimitDegrees=45
  twistMotion=Free
 }
 {
  bone1Name=clavicle_r
  bone2Name=spine_05
  swing1LimitDegrees=45
  swing1Motion=Free
  swing2LimitDegrees=45
  swing2Motion=Free
  twistLimitDegrees=45
  twistMotion=Free
 }
 {
  bone1Name=dyn_body_zipper_l_MSTR
  bone2Name=spine_05
  swing1LimitDegrees=45
  swing1Motion=Free
  swing2LimitDegrees=45
  swing2Motion=Free
  twistLimitDegrees=45
  twistMotion=Free
 }
 {
  bone1Name=dyn_body_zipper_r_MSTR
  bone2Name=spine_05
  swing1LimitDegrees=45
  swing1Motion=Free
  swing2LimitDegrees=45
  swing2Motion=Free
  twistLimitDegrees=45
  twistMotion=Free
 }
 {bone1Name=dyn_body_zipper_l_01, bone2Name=dyn_body_zipper_l_MSTR, swing1LimitDegrees=25, swing2LimitDegrees=25, twistLimitDegrees=2}
 {bone1Name=dyn_body_zipper_r_01, bone2Name=dyn_body_zipper_r_MSTR, swing1LimitDegrees=25, swing2LimitDegrees=25, twistLimitDegrees=2}
 {
  bone1Name=dyn_hood_MSTR
  bone2Name=spine_05
  swing1LimitDegrees=45
  swing1Motion=Free
  swing2LimitDegrees=45
  swing2Motion=Free
  twistLimitDegrees=45
  twistMotion=Free
 }
 {bone1Name=dyn_hood_01, bone2Name=dyn_hood_MSTR, swing1LimitDegrees=10, swing2LimitDegrees=25, twistLimitDegrees=4}
 {
  bone1Name=dyn_coat_fr_l_MSTR
  bone2Name=spine_04
  swing1LimitDegrees=45
  swing1Motion=Free
  swing2LimitDegrees=45
  swing2Motion=Free
  twistLimitDegrees=45
  twistMotion=Free
 }
 {bone1Name=dyn_coat_fr_l_01, bone2Name=dyn_coat_fr_l_MSTR, swing1LimitDegrees=45, swing2LimitDegrees=45, twistLimitDegrees=2}
 {
  bone1Name=dyn_coat_bk_l_MSTR
  bone2Name=spine_04
  swing1LimitDegrees=45
  swing1Motion=Free
  swing2LimitDegrees=45
  swing2Motion=Free
  twistLimitDegrees=45
  twistMotion=Free
 }
 {bone1Name=dyn_coat_bk_l_01, bone2Name=dyn_coat_bk_l_MSTR, swing1LimitDegrees=35, swing2LimitDegrees=35, twistLimitDegrees=2}
 {
  bone1Name=dyn_coat_fr_r_MSTR
  bone2Name=spine_04
  swing1LimitDegrees=45
  swing1Motion=Free
  swing2LimitDegrees=45
  swing2Motion=Free
  twistLimitDegrees=45
  twistMotion=Free
 }
 {bone1Name=dyn_coat_fr_r_01, bone2Name=dyn_coat_fr_r_MSTR, swing1LimitDegrees=45, swing2LimitDegrees=45, twistLimitDegrees=2}
 {
  bone1Name=dyn_coat_bk_r_MSTR
  bone2Name=spine_04
  swing1LimitDegrees=45
  swing1Motion=Free
  swing2LimitDegrees=45
  swing2Motion=Free
  twistLimitDegrees=45
  twistMotion=Free
 }
 {bone1Name=dyn_coat_bk_r_01, bone2Name=dyn_coat_bk_r_MSTR, swing1LimitDegrees=35, swing2LimitDegrees=35, twistLimitDegrees=2}
 {
  bone1Name=upperarm_twist_01_l
  bone2Name=clavicle_l
  swing1LimitDegrees=45
  swing1Motion=Free
  swing2LimitDegrees=45
  swing2Motion=Free
  twistLimitDegrees=45
  twistMotion=Free
 }
 {
  bone1Name=upperarm_twist_01_r
  bone2Name=clavicle_r
  swing1LimitDegrees=45
  swing1Motion=Free
  swing2LimitDegrees=45
  swing2Motion=Free
  twistLimitDegrees=45
  twistMotion=Free
 }
 {
  bone1Name=neck_01
  bone2Name=spine_05
  swing1LimitDegrees=45
  swing1Motion=Free
  swing2LimitDegrees=45
  swing2Motion=Free
  twistLimitDegrees=45
  twistMotion=Free
 }
 {bone1Name=dyn_scarf_lining_r_MSTR, bone2Name=neck_01, swing1LimitDegrees=15, swing2LimitDegrees=15, twistLimitDegrees=0}
 {bone1Name=dyn_scarf_lining_l_MSTR, bone2Name=neck_01, swing1LimitDegrees=15, swing2LimitDegrees=15, twistLimitDegrees=0}
]
