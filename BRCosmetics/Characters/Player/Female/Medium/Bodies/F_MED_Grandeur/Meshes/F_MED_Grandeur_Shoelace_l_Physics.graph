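# Physics asset for the left boot shoelace of the female medium Grandeur player body.
# Capsules on pelvis, thigh, calf, calf twist and foot give the leg's shape alongside small capsules on the lace and knot bones.
# The knot master hangs off the calf twist with free angular motion, then chains to a knot that forks into two-segment left and right laces.
# Knot and lace joints use tight swing limits and almost no twist.
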PhysicsAsset /BRCosmetics/Characters/Player/Female/Medium/Bodies/F_MED_Grandeur/Meshes/F_MED_Grandeur_Shoelace_l_Physics
bodies {
 calf_l [{shapeType=Capsule, length=24.5136, radius=8, center=<-23.8948 -2.7991 -0.0002>, rotation=<rot 90 0 0>}]
 calf_twist_01_l [{
   shapeType=Capsule
   length=5.9735
   radius=5.2202
   shapeName=calf_twist_01_l_capsule
   center=<-1.3503 -0.7913 -0.3585>
   rotation=<rot 0 141.0151 -82.9253>
  }]
 dyn_boot_l_knot_01 [{shapeType=Capsule, length=0.505, radius=0.505, shapeName=dyn_boot_l_knot_01_capsule, center=<0.5 0 0>, rotation=<rot 90 0 0>}]
 dyn_boot_l_knot_MSTR [{shapeType=Capsule, length=0.505, radius=0.505, shapeName=dyn_boot_l_knot_MSTR_capsule, rotation=<rot 0 -90 0>}]
 dyn_boot_l_lace_l_01 [{shapeType=Capsule, length=3.8986, radius=0.7572, shapeName=dyn_boot_l_lace_l_01_capsule, center=<0.5 0 0>, rotation=<rot 90 0 0>}]
 dyn_boot_l_lace_l_02 [{shapeType=Capsule, length=1.0905, radius=0.6228, shapeName=dyn_boot_l_lace_l_02_capsule, center=<0.5 0 0>, rotation=<rot 90 0 0>}]
 dyn_boot_l_lace_r_01 [{shapeType=Capsule, length=3.6706, radius=0.6542, shapeName=dyn_boot_l_lace_r_01_capsule, center=<0.5 0 0>, rotation=<rot 90 0 0>}]
 dyn_boot_l_lace_r_02 [{shapeType=Capsule, length=1.0913, radius=0.6389, shapeName=dyn_boot_l_lace_r_02_capsule, center=<0.5 0 0>, rotation=<rot 90 0 0>}]
 dyn_boot_l_top_lace_MSTR [{shapeType=Capsule, length=0.505, radius=0.505, shapeName=dyn_boot_l_top_lace_MSTR_capsule, rotation=<rot 0 -90 0>}]
 foot_l [{
   shapeType=Capsule
   length=13.3939
   radius=5.3493
   shapeName=foot_r_capsule
   center=<-4.5441 6.1999 0.0962>
   rotation=<rot 0.1489 -168.0228 -91.703>
  }]
 pelvis [{shapeType=Capsule, length=16.9528, radius=14.1111, center=<1.1484 -1.7159 -1.3183>, rotation=<rot 0 178.4668 -147.7398>}]
 thigh_l [{shapeType=Capsule, length=20.2024, radius=9.7595, center=<-23.1357 0 0>, rotation=<rot 90 0 0>}]
}
constraints [
 {bone1Name=thigh_l, bone2Name=pelvis, swing1LimitDegrees=45, swing2LimitDegrees=45, twistLimitDegrees=45}
 {bone1Name=calf_l, bone2Name=thigh_l, swing1LimitDegrees=45, swing2LimitDegrees=45, twistLimitDegrees=45}
 {
  bone1Name=calf_twist_01_l
  bone2Name=calf_l
  swing1LimitDegrees=45
  swing1Motion=Free
  swing2LimitDegrees=45
  swing2Motion=Free
  twistLimitDegrees=45
  twistMotion=Free
 }
 {
  bone1Name=dyn_boot_l_top_lace_MSTR
  bone2Name=calf_twist_01_l
  swing1LimitDegrees=45
  swing1Motion=Free
  swing2LimitDegrees=45
  swing2Motion=Free
  twistLimitDegrees=45
  twistMotion=Free
 }
 {
  bone1Name=dyn_boot_l_knot_MSTR
  bone2Name=calf_twist_01_l
  swing1LimitDegrees=45
  swing1Motion=Free
  swing2LimitDegrees=45
  swing2Motion=Free
  twistLimitDegrees=45
  twistMotion=Free
 }
 {bone1Name=dyn_boot_l_knot_01, bone2Name=dyn_boot_l_knot_MSTR, swing1LimitDegrees=20, swing2LimitDegrees=20, twistLimitDegrees=2}
 {bone1Name=dyn_boot_l_lace_r_01, bone2Name=dyn_boot_l_knot_01, swing1LimitDegrees=25, swing2LimitDegrees=25, twistLimitDegrees=2}
 {bone1Name=dyn_boot_l_lace_r_02, bone2Name=dyn_boot_l_lace_r_01, swing1LimitDegrees=25, swing2LimitDegrees=25, twistLimitDegrees=2}
 {bone1Name=dyn_boot_l_lace_l_01, bone2Name=dyn_boot_l_knot_01, swing1LimitDegrees=25, swing2LimitDegrees=25, twistLimitDegrees=2}
 {bone1Name=dyn_boot_l_lace_l_02, bone2Name=dyn_boot_l_lace_l_01, swing1LimitDegrees=25, swing2LimitDegrees=25, twistLimitDegrees=2}
 {
  bone1Name=foot_l
  bone2Name=calf_l
  swing1LimitDegrees=45
  swing1Motion=Free
  swing2LimitDegrees=45
  swing2Motion=Free
  twistLimitDegrees=45
  twistMotion=Free
 }
]
