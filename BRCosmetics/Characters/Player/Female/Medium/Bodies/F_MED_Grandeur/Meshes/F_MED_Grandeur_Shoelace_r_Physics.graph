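# Physics asset for the right boot shoelaces of the F_MED_Grandeur female player body, built entirely from capsule bodies.
# Pelvis, thigh, calf, calf twist and foot capsules form the leg chain, with knot and top-lace master bodies attached to the calf twist by free-rotation constraints.
# Below the knot master, a knot body branches into two-segment left and right lace chains, with limited swing and very tight twist.
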PhysicsAsset /BRCosmetics/Characters/Player/Female/Medium/Bodies/F_MED_Grandeur/Meshes/F_MED_Grandeur_Shoelace_r_Physics
bodies {
 calf_r [{shapeType=Capsule, length=24.3162, radius=8, center=<24.1011 2.7991 0>, rotation=<rot 90 0 0>}]
 calf_twist_01_r [{
   shapeType=Capsule
   length=5.9559
   radius=5.2172
   shapeName=calf_twist_01_r_capsule
   center=<1.3454 0.7892 0.347>
   rotation=<rot 0 141.1945 -83.8681>
  }]
 dyn_boot_r_knot_01 [{
   shapeType=Capsule
   length=1.2211
   radius=1.0864
   shapeName=dyn_boot_r_knot_01_capsule
   center=<0.5961 -0.082 0.0724>
   rotation=<rot 0 60.0264 -88.582>
  }]
 dyn_boot_r_knot_MSTR [{shapeType=Capsule, length=0.505, radius=0.505, shapeName=dyn_boot_r_knot_MSTR_capsule, rotation=<rot 0 -90 0>}]
 dyn_boot_r_lace_l_01 [{
   shapeType=Capsule
   length=3.6706
   radius=0.6542
   shapeName=dyn_boot_r_lace_l_01_capsule
   center=<1.8911 -0.254 -0.0589>
   rotation=<rot 0 -99.0569 -89.6974>
  }]
 dyn_boot_r_lace_l_02 [{
   shapeType=Capsule
   length=1.0913
   radius=0.6389
   shapeName=dyn_boot_r_lace_l_02_capsule
   center=<0.978 0.0187 -0.0611>
   rotation=<rot 0 92.464 -87.6058>
  }]
 dyn_boot_r_lace_r_01 [{
   shapeType=Capsule
   length=3.8986
   radius=0.7572
   shapeName=dyn_boot_r_lace_r_01_capsule
   center=<2.1148 -0.3587 0.0378>
   rotation=<rot 0 78.1817 -89.0611>
  }]
 dyn_boot_r_lace_r_02 [{
   shapeType=Capsule
   length=1.0905
   radius=0.6229
   shapeName=dyn_boot_r_lace_r_02_capsule
   center=<1.0577 -0.0579 0.0627>
   rotation=<rot 0 -88.389 -87.8616>
  }]
 dyn_boot_r_top_lace_MSTR [{shapeType=Capsule, length=0.505, radius=0.505, shapeName=dyn_boot_r_top_lace_MSTR_capsule, rotation=<rot 0 -90 0>}]
 foot_r [{
   shapeType=Capsule
   length=13.3939
   radius=5.3493
   shapeName=foot_r_capsule
   center=<4.5441 -6.1999 0.0962>
   rotation=<rot 0.1489 -168.0228 -91.703>
  }]
 pelvis [{shapeType=Capsule, length=16.9528, radius=14.1111, center=<1.1484 -1.7159 -1.3183>, rotation=<rot 0 178.4668 -147.7398>}]
 thigh_r [{shapeType=Capsule, length=20.522, radius=10.6123, center=<22.9295 0 0>, rotation=<rot 90 0 0>}]
}
constraints [
 {bone1Name=thigh_r, bone2Name=pelvis, swing1LimitDegrees=45, swing2LimitDegrees=45, twistLimitDegrees=45}
 {bone1Name=calf_r, bone2Name=thigh_r, swing1LimitDegrees=45, swing2LimitDegrees=45, twistLimitDegrees=45}
 {
  bone1Name=calf_twist_01_r
  bone2Name=calf_r
  swing1LimitDegrees=45
  swing1Motion=Free
  swing2LimitDegrees=45
  swing2Motion=Free
  twistLimitDegrees=45
  twistMotion=Free
 }
 {
  bone1Name=dyn_boot_r_top_lace_MSTR
  bone2Name=calf_twist_01_r
  swing1LimitDegrees=45
  swing1Motion=Free
  swing2LimitDegrees=45
  swing2Motion=Free
  twistLimitDegrees=45
  twistMotion=Free
 }
 {
  bone1Name=dyn_boot_r_knot_MSTR
  bone2Name=calf_twist_01_r
  swing1LimitDegrees=45
  swing1Motion=Free
  swing2LimitDegrees=45
  swing2Motion=Free
  twistLimitDegrees=45
  twistMotion=Free
 }
 {bone1Name=dyn_boot_r_knot_01, bone2Name=dyn_boot_r_knot_MSTR, swing1LimitDegrees=45, swing2LimitDegrees=45, twistLimitDegrees=2}
 {bone1Name=dyn_boot_r_lace_l_01, bone2Name=dyn_boot_r_knot_01, swing1LimitDegrees=25, swing2LimitDegrees=25, twistLimitDegrees=2}
 {bone1Name=dyn_boot_r_lace_l_02, bone2Name=dyn_boot_r_lace_l_01, swing1LimitDegrees=25, swing2LimitDegrees=25, twistLimitDegrees=2}
 {bone1Name=dyn_boot_r_lace_r_01, bone2Name=dyn_boot_r_knot_01, swing1LimitDegrees=25, swing2LimitDegrees=25, twistLimitDegrees=2}
 {bone1Name=dyn_boot_r_lace_r_02, bone2Name=dyn_boot_r_lace_r_01, swing1LimitDegrees=25, swing2LimitDegrees=25, twistLimitDegrees=2}
 {
  bone1Name=foot_r
  bone2Name=calf_r
  swing1LimitDegrees=45
  swing1Motion=Free
  swing2LimitDegrees=45
  swing2Motion=Free
  twistLimitDegrees=45
  twistMotion=Free
 }
]
