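# Cosmetic AnimBP for the F_MED_Grandeur body, adding secondary motion to clothing and accessories.
# Parent-mesh pose passes through a Control Rig, bone constraints, Anim Dynamics and three RigidBody simulations.
# Additive bone offsets then correct hood, straps, belts and bag, driven by movement states and head, arm and thigh rotations measured each update; skydiving also repositions the backpack.
# A separate boot-lace Anim Dynamics chain is disconnected from the output.
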AnimBlueprint /BRCosmetics/Characters/Player/Female/Medium/Bodies/F_MED_Grandeur/Meshes/F_MED_Grandeur_AnimBP
target_skeleton=&/F_MED_Grandeur_Skeleton
graphs [
 AnimGraph {nodes [
   n1 = Anim Dynamics - Bone: dyn_boot_l_lace_r_01(Alpha=1, ExternalForce="0,0,0") {
    angularDampingOverride=0.999
    bGravityOverrideInSimSpace=true
    bOverrideAngularDamping=true
    bUseGravityOverride=true
    bUseSphericalLimits=true
    alphaBoolBlend {blendOption=Linear}
    boundBone {boneName=dyn_boot_l_lace_r_01}
    componentAppliedLinearAccClamp=<3000 3000 3000>
    componentLinearAccScale=<0.3 0.3 0.3>
    componentLinearVelScale=<0.1 0.1 0.1>
    gravityOverride {z=-980}
    physicsBodyDefinitions [{
      sphereCollisionRadius=0
      boundBone {boneName=dyn_boot_l_lace_r_01}
      boxExtents=<12 6 12>
      constraintSetup {angularLimitsMax=<0 15 10>, angularLimitsMin=<0 -15 -2>, angularTarget=<0 0 0>}
      localJointOffset=<1 0 0>
     }]
    relativeSpaceBone {boneName=dyn_boot_l_knot_01}
    sphericalLimits [{limitRadius=5, drivingBone {boneName=foot_l}, sphereLocalOffset=<0 2 0>}]
   }
   n2 = Anim Dynamics - Bone: dyn_boot_l_lace_r_02(ComponentPose=@n1, Alpha=1, ExternalForce="0,0,0") {
    angularDampingOverride=0.999
    bGravityOverrideInSimSpace=true
    bOverrideAngularDamping=true
    bUseGravityOverride=true
    bUseSphericalLimits=true
    alphaBoolBlend {blendOption=Linear}
    boundBone {boneName=dyn_boot_l_lace_r_02}
    componentAppliedLinearAccClamp=<3000 3000 3000>
    componentLinearAccScale=<0.3 0.3 0.3>
    componentLinearVelScale=<0.1 0.1 0.1>
    gravityOverride {z=-980}
    physicsBodyDefinitions [{
      sphereCollisionRadius=0
      boundBone {boneName=dyn_boot_l_lace_r_02}
      boxExtents=<12 6 12>
      constraintSetup {angularLimitsMax=<0 15 10>, angularLimitsMin=<0 -15 -2>, angularTarget=<0 0 0>}
      localJointOffset=<1 0 0>
     }]
    relativeSpaceBone {boneName=dyn_boot_l_lace_r_01}
    sphericalLimits [{limitRadius=5, drivingBone {boneName=foot_l}, sphereLocalOffset=<0 2 0>}]
   }
   n3 = Anim Dynamics - Bone: dyn_boot_l_lace_l_01(ComponentPose=@n2, Alpha=1, ExternalForce="0,0,0") {
    angularDampingOverride=0.999
    bGravityOverrideInSimSpace=true
    bOverrideAngularDamping=true
    bUseGravityOverride=true
    bUseSphericalLimits=true
    alphaBoolBlend {blendOption=Linear}
    boundBone {boneName=dyn_boot_l_lace_l_01}
    componentAppliedLinearAccClamp=<3000 3000 3000>
    componentLinearAccScale=<0.3 0.3 0.3>
    componentLinearVelScale=<0.1 0.1 0.1>
    gravityOverride {z=-980}
    physicsBodyDefinitions [{
      sphereCollisionRadius=0
      boundBone {boneName=dyn_boot_l_lace_l_01}
      boxExtents=<12 6 12>
      constraintSetup {angularLimitsMax=<0 15 10>, angularLimitsMin=<0 -15 -2>, angularTarget=<0 0 0>}
      localJointOffset=<1 0 0>
     }]
    relativeSpaceBone {boneName=dyn_boot_l_knot_01}
    sphericalLimits [{limitRadius=5, drivingBone {boneName=foot_l}, sphereLocalOffset=<0 2 0>}]
   }
   n4 = Anim Dynamics - Bone: dyn_boot_l_lace_l_02(ComponentPose=@n3, Alpha=1, ExternalForce="0,0,0") {
    angularDampingOverride=0.999
    bGravityOverrideInSimSpace=true
    bOverrideAngularDamping=true
    bUseGravityOverride=true
    bUseSphericalLimits=true
    alphaBoolBlend {blendOption=Linear}
    boundBone {boneName=dyn_boot_l_lace_l_02}
    componentAppliedLinearAccClamp=<3000 3000 3000>
    componentLinearAccScale=<0.3 0.3 0.3>
    componentLinearVelScale=<0.1 0.1 0.1>
    gravityOverride {z=-980}
    physicsBodyDefinitions [{
      sphereCollisionRadius=0
      boundBone {boneName=dyn_boot_l_lace_l_02}
      boxExtents=<12 6 12>
      constraintSetup {angularLimitsMax=<0 15 10>, angularLimitsMin=<0 -15 -2>, angularTarget=<0 0 0>}
      localJointOffset=<1 0 0>
     }]
    relativeSpaceBone {boneName=dyn_boot_l_lace_l_01}
    sphericalLimits [{limitRadius=5, drivingBone {boneName=foot_l}, sphereLocalOffset=<0 2 0>}]
   }
   n5 = Anim Dynamics - Bone: dyn_boot_r_lace_l_01(ComponentPose=@n4, Alpha=1, ExternalForce="0,0,0") {
    angularDampingOverride=0.999
    bGravityOverrideInSimSpace=true
    bOverrideAngularDamping=true
    bUseGravityOverride=true
    bUseSphericalLimits=true
    alphaBoolBlend {blendOption=Linear}
    boundBone {boneName=dyn_boot_r_lace_l_01}
    componentAppliedLinearAccClamp=<3000 3000 3000>
    componentLinearAccScale=<0.3 0.3 0.3>
    componentLinearVelScale=<0.1 0.1 0.1>
    gravityOverride {z=-980}
    physicsBodyDefinitions [{
      sphereCollisionRadius=0
      boundBone {boneName=dyn_boot_r_lace_l_01}
      boxExtents=<12 6 12>
      constraintSetup {angularLimitsMax=<0 15 10>, angularLimitsMin=<0 -15 0>, angularTarget=<0 0 0>}
      localJointOffset=<1 0 0>
     }]
    relativeSpaceBone {boneName=dyn_boot_r_knot_01}
    sphericalLimits [{limitRadius=5, drivingBone {boneName=foot_r}, sphereLocalOffset=<0 -2 0>}]
   }
   n6 = Anim Dynamics - Bone: dyn_boot_r_lace_l_02(ComponentPose=@n5, Alpha=1, ExternalForce="0,0,0") {
    angularDampingOverride=0.999
    bGravityOverrideInSimSpace=true
    bOverrideAngularDamping=true
    bUseGravityOverride=true
    bUseSphericalLimits=true
    alphaBoolBlend {blendOption=Linear}
    boundBone {boneName=dyn_boot_r_lace_l_02}
    componentAppliedLinearAccClamp=<3000 3000 3000>
    componentLinearAccScale=<0.3 0.3 0.3>
    componentLinearVelScale=<0.1 0.1 0.1>
    gravityOverride {z=-980}
    physicsBodyDefinitions [{
      sphereCollisionRadius=0
      boundBone {boneName=dyn_boot_r_lace_l_02}
      boxExtents=<12 6 12>
      constraintSetup {angularLimitsMax=<0 15 10>, angularLimitsMin=<0 -15 0>, angularTarget=<0 0 0>}
      localJointOffset=<1 0 0>
     }]
    relativeSpaceBone {boneName=dyn_boot_r_lace_l_01}
    sphericalLimits [{limitRadius=5, drivingBone {boneName=foot_r}, sphereLocalOffset=<0 -2 0>}]
   }
   n7 = Anim Dynamics - Bone: dyn_boot_r_lace_r_01(ComponentPose=@n6, Alpha=1, ExternalForce="0,0,0") {
    angularDampingOverride=0.999
    bGravityOverrideInSimSpace=true
    bOverrideAngularDamping=true
    bUseGravityOverride=true
    bUseSphericalLimits=true
    alphaBoolBlend {blendOption=Linear}
    boundBone {boneName=dyn_boot_r_lace_r_01}
    componentAppliedLinearAccClamp=<3000 3000 3000>
    componentLinearAccScale=<0.3 0.3 0.3>
    componentLinearVelScale=<0.1 0.1 0.1>
    gravityOverride {z=-980}
    physicsBodyDefinitions [{
      sphereCollisionRadius=0
      boundBone {boneName=dyn_boot_r_lace_r_01}
      boxExtents=<12 6 12>
      constraintSetup {angularLimitsMax=<0 15 10>, angularLimitsMin=<0 -15 0>, angularTarget=<0 0 0>}
      localJointOffset=<1 0 0>
     }]
    relativeSpaceBone {boneName=dyn_boot_r_knot_01}
    sphericalLimits [{limitRadius=5, drivingBone {boneName=foot_r}, sphereLocalOffset=<0 -2 0>}]
   }
   Anim Dynamics - Bone: dyn_boot_r_lace_r_02(ComponentPose=@n7, Alpha=1, ExternalForce="0,0,0") {
    angularDampingOverride=0.999
    bGravityOverrideInSimSpace=true
    bOverrideAngularDamping=true
    bUseGravityOverride=true
    bUseSphericalLimits=true
    alphaBoolBlend {blendOption=Linear}
    boundBone {boneName=dyn_boot_r_lace_r_02}
    componentAppliedLinearAccClamp=<3000 3000 3000>
    componentLinearAccScale=<0.3 0.3 0.3>
    componentLinearVelScale=<0.1 0.1 0.1>
    gravityOverride {z=-980}
    physicsBodyDefinitions [{
      sphereCollisionRadius=0
      boundBone {boneName=dyn_boot_r_lace_r_02}
      boxExtents=<12 6 12>
      constraintSetup {angularLimitsMax=<0 15 10>, angularLimitsMin=<0 -15 0>, angularTarget=<0 0 0>}
      localJointOffset=<1 0 0>
     }]
    relativeSpaceBone {boneName=dyn_boot_r_lace_r_01}
    sphericalLimits [{limitRadius=5, drivingBone {boneName=foot_r}, sphereLocalOffset=<0 -2 0>}]
   }
   n8 = Control Rig(Source=Copy Pose From Mesh() {bUseAttachedParent=true}) {
    alphaBoolBlend {blendOption=Linear}
    controlRigAssetReference {blueprintRigClass=&/F_MED_Grandeur_CtrlRig.F_MED_Grandeur_CtrlRig_C}
   }
   n9 = "Constraint - dyn_belt_leg_r_MSTR (PPPPPPPPP)"!(
    ComponentPose=Local To Component(LocalPose=@n8)
    Alpha=1
    ConstraintWeights_0=0.0001
    ConstraintWeights_1:float=0.0052
    ConstraintWeights_2:float=0.2554
    ConstraintWeights_3:float=0.5664
    ConstraintWeights_4:float=0.0048
    ConstraintWeights_5:float=0.0001
    ConstraintWeights_6:float=0.1313
    ConstraintWeights_7:float=0
    ConstraintWeights_8:float=0.0368
   ) -> (Pose) {
    alphaBoolBlend {blendOption=Linear}
    boneToModify {boneName=dyn_belt_leg_r_MSTR}
    constraintSetup [
     {targetBone {boneName=deform_thigh_fr_l}}
     {targetBone {boneName=deform_thigh_bk_r}}
     {targetBone {boneName=thigh_r}}
     {targetBone {boneName=thigh_twist_01_r}}
     {targetBone {boneName=thigh_twist_01_l}}
     {targetBone {boneName=thigh_l}}
     {targetBone {boneName=pelvis}}
     {targetBone {boneName=deform_thigh_bk_l}}
     {targetBone {boneName=deform_thigh_fr_r}}
    ]
    constraintWeights [1,1,1,1,1,1,1,1,1]
   }
   n10 = "Constraint - dyn_body_zipper_r_MSTR (PPP)"!(ComponentPose=@n9, Alpha=1, ConstraintWeights_0=0.7252, ConstraintWeights_1:float=0.2584, ConstraintWeights_2:float=0.0164) -> (Pose) {
    alphaBoolBlend {blendOption=Linear}
    boneToModify {boneName=dyn_body_zipper_r_MSTR}
    constraintSetup [{targetBone {boneName=spine_04}}, {targetBone {boneName=spine_05}}, {targetBone {boneName=deform_pectoral_fr_r}}]
    constraintWeights [1,1,1]
   }
   n11 = "Constraint - dyn_body_zipper_l_MSTR (PPP)"!(ComponentPose=@n10, Alpha=1, ConstraintWeights_0=0.0164, ConstraintWeights_1:float=0.7254, ConstraintWeights_2:float=0.2582) -> (Pose) {
    alphaBoolBlend {blendOption=Linear}
    boneToModify {boneName=dyn_body_zipper_l_MSTR}
    constraintSetup [{targetBone {boneName=deform_pectoral_fr_l}}, {targetBone {boneName=spine_04}}, {targetBone {boneName=spine_05}}]
    constraintWeights [1,1,1]
   }
   n12 = "Constraint - dyn_belt_back_MSTR (P)"!(ComponentPose=@n11, Alpha=1, ConstraintWeights_0=1, ConstraintWeights_1:float=0) -> (Pose) {
    alphaBoolBlend {blendOption=Linear}
    boneToModify {boneName=dyn_belt_back_MSTR}
    constraintSetup [{targetBone {boneName=pelvis}}]
    constraintWeights [1,0]
   }
   n13 = "Constraint - dyn_boot_l_top_lace_MSTR (PPPP)"!(
    ComponentPose=@n12
    Alpha=1
    ConstraintWeights_0=0.0787
    ConstraintWeights_1:float=0.5767
    ConstraintWeights_2:float=0.0392
    ConstraintWeights_3:float=0.3053
   ) -> (Pose) {
    alphaBoolBlend {blendOption=Linear}
    boneToModify {boneName=dyn_boot_l_top_lace_MSTR}
    constraintSetup [
     {targetBone {boneName=calf_l}}
     {targetBone {boneName=calf_twist_01_l}}
     {targetBone {boneName=foot_l}}
     {targetBone {boneName=calf_twist_02_l}}
    ]
    constraintWeights [1,1,1,1]
   }
   n14 = "Constraint - dyn_boot_r_top_lace_MSTR (PPPP)"!(
    ComponentPose=@n13
    Alpha=1
    ConstraintWeights_0=0.3053
    ConstraintWeights_1:float=0.5767
    ConstraintWeights_2:float=0.0392
    ConstraintWeights_3:float=0.0787
   ) -> (Pose) {
    alphaBoolBlend {blendOption=Linear}
    boneToModify {boneName=dyn_boot_r_top_lace_MSTR}
    constraintSetup [
     {targetBone {boneName=calf_twist_02_r}}
     {targetBone {boneName=calf_twist_01_r}}
     {targetBone {boneName=foot_r}}
     {targetBone {boneName=calf_r}}
    ]
    constraintWeights [1,1,1,1]
   }
   n15 = "Constraint - dyn_boot_l_knot_MSTR (PPPP)"!(
    ComponentPose=@n14
    Alpha=1
    ConstraintWeights_0=0.0596
    ConstraintWeights_1:float=0.6823
    ConstraintWeights_2:float=0.0841
    ConstraintWeights_3:float=0.1739
   ) -> (Pose) {
    alphaBoolBlend {blendOption=Linear}
    boneToModify {boneName=dyn_boot_l_knot_MSTR}
    constraintSetup [
     {targetBone {boneName=calf_l}}
     {targetBone {boneName=calf_twist_01_l}}
     {targetBone {boneName=foot_l}}
     {targetBone {boneName=calf_twist_02_l}}
    ]
    constraintWeights [1,1,1,1]
   }
   n16 = "Constraint - dyn_boot_r_knot_MSTR (PPPP)"!(
    ComponentPose=@n15
    Alpha=1
    ConstraintWeights_0=0.1739
    ConstraintWeights_1:float=0.6823
    ConstraintWeights_2:float=0.0841
    ConstraintWeights_3:float=0.0596
   ) -> (Pose) {
    alphaBoolBlend {blendOption=Linear}
    boneToModify {boneName=dyn_boot_r_knot_MSTR}
    constraintSetup [
     {targetBone {boneName=calf_twist_02_r}}
     {targetBone {boneName=calf_twist_01_r}}
     {targetBone {boneName=foot_r}}
     {targetBone {boneName=calf_r}}
    ]
    constraintWeights [1,1,1,1]
   }
   n17 = Anim Dynamics - Bone: dyn_belt_leg_r_01(ComponentPose=@n16, Alpha=SelectFloat(A=0.3, B=1, bPickA=Get bIsCrouching()), ExternalForce="0,0,0") {
    angularDampingOverride=0.998
    bGravityOverrideInSimSpace=true
    bOverrideAngularDamping=true
    bUseGravityOverride=true
    simulationSpace=RootRelative
    alphaBoolBlend {blendOption=Linear}
    boundBone {boneName=dyn_belt_leg_r_01}
    componentAppliedLinearAccClamp=<3000 3000 3000>
    componentLinearAccScale=<0.3 0.3 0.27>
    componentLinearVelScale=<0.2 0.2 0.2>
    gravityOverride {z=-540}
    physicsBodyDefinitions [{
      sphereCollisionRadius=0
      boundBone {boneName=dyn_belt_leg_r_01}
      boxExtents=<12 6 12>
      constraintSetup {angularLimitsMax=<0 20 0>, angularLimitsMin=<0 -20 -25>, angularTarget=<0 0 0>}
      localJointOffset=<2 0 0>
     }]
    relativeSpaceBone {boneName=pelvis}
   }
   n18 = Anim Dynamics - Bone: dyn_puffJacket_sleeve_r(ComponentPose=@n17, Alpha=0.5, ExternalForce="0,0,0") {
    angularDampingOverride=0.998
    bGravityOverrideInSimSpace=true
    bOverrideAngularDamping=true
    bUseGravityOverride=true
    simulationSpace=RootRelative
    alphaBoolBlend {blendOption=Linear}
    boundBone {boneName=dyn_puffJacket_sleeve_r}
    componentAppliedLinearAccClamp=<3000 3000 3000>
    componentLinearAccScale=<0.3 0.3 0.3>
    componentLinearVelScale=<0.3 0.3 0.3>
    gravityOverride {z=-980}
    physicsBodyDefinitions [{
      sphereCollisionRadius=0
      boundBone {boneName=dyn_puffJacket_sleeve_r}
      boxExtents=<12 6 12>
      constraintSetup {angularLimitsMax=<0 5 5>, angularLimitsMin=<0 -5 -5>, angularTarget=<0 0 0>, linearAxesMax=<0 1 1>, linearAxesMin=<0 -1 -1>}
      localJointOffset=<1 0 0>
     }]
    relativeSpaceBone {boneName=lowerarm_r}
   }
   n19 = Anim Dynamics - Bone: dyn_puffJacket_sleeve_l(ComponentPose=@n18, Alpha=0.5, ExternalForce="0,0,0") {
    angularDampingOverride=0.998
    bGravityOverrideInSimSpace=true
    bOverrideAngularDamping=true
    bUseGravityOverride=true
    simulationSpace=RootRelative
    alphaBoolBlend {blendOption=Linear}
    boundBone {boneName=dyn_puffJacket_sleeve_l}
    componentAppliedLinearAccClamp=<3000 3000 3000>
    componentLinearAccScale=<0.3 0.3 0.3>
    componentLinearVelScale=<0.3 0.3 0.3>
    gravityOverride {z=-980}
    physicsBodyDefinitions [{
      sphereCollisionRadius=0
      boundBone {boneName=dyn_puffJacket_sleeve_l}
      boxExtents=<12 6 12>
      constraintSetup {angularLimitsMax=<0 5 5>, angularLimitsMin=<0 -5 -5>, angularTarget=<0 0 0>, linearAxesMax=<0 1 1>, linearAxesMin=<0 -1 -1>}
      localJointOffset=<1 0 0>
     }]
    relativeSpaceBone {boneName=lowerarm_l}
   }
   n20 = Anim Dynamics - Bone: dyn_belt_back_01(ComponentPose=@n19, Alpha=1, ExternalForce="0,0,0") {
    angularDampingOverride=0.998
    bGravityOverrideInSimSpace=true
    bOverrideAngularDamping=true
    bUseGravityOverride=true
    simulationSpace=RootRelative
    alphaBoolBlend {blendOption=Linear}
    boundBone {boneName=dyn_belt_back_01}
    componentAppliedLinearAccClamp=<3000 3000 3000>
    componentLinearAccScale=<0.3 0.3 0.3>
    componentLinearVelScale=<0.3 0.3 0.3>
    gravityOverride {z=-980}
    physicsBodyDefinitions [{
      sphereCollisionRadius=0
      boundBone {boneName=dyn_belt_back_01}
      boxExtents=<15 15 7>
      constraintSetup {angularLimitsMax=<0 0 25>, angularLimitsMin=<0 0 -25>, angularTarget=<0 0 0>}
      localJointOffset=<1.5 -0.1 0>
     }]
    relativeSpaceBone {boneName=dyn_belt_back_MSTR}
   }
   n21 = Anim Dynamics - Bone: dyn_bag_l_01(ComponentPose=@n20, Alpha=0.6, ExternalForce="0,0,0") {
    angularDampingOverride=0.998
    bGravityOverrideInSimSpace=true
    bOverrideAngularDamping=true
    bUseGravityOverride=true
    simulationSpace=RootRelative
    alphaBoolBlend {blendOption=Linear}
    boundBone {boneName=dyn_bag_l_01}
    componentAppliedLinearAccClamp=<3000 3000 3000>
    componentLinearAccScale=<0.2 0.2 0.2>
    componentLinearVelScale=<0.1 0.1 0.1>
    gravityOverride {z=-980}
    physicsBodyDefinitions [{
      sphereCollisionRadius=0
      boundBone {boneName=dyn_bag_l_01}
      boxExtents=<12 12 8>
      constraintSetup {angularLimitsMax=<2 5 3>, angularLimitsMin=<-2 -5 -3>, angularTarget=<0 0 0>}
      localJointOffset=<0.5 0 0>
     }]
    relativeSpaceBone {boneName=pelvis}
   }
   n22 = SelectFloat(A=0.5, B=1, bPickA=Get bIsSkyDiving())
   n23 = RigidBody(ComponentPose=@n21, Alpha=@n22) {
    bClampLinearTranslationLimitToRefPose=true
    bOverrideWorldGravity=true
    overridePhysicsAsset=&/F_MED_Grandeur_Physics
    simulationSpace=BaseBoneSpace
    alphaBoolBlend {blendOption=Linear}
    baseBoneRef {boneName=spine_03}
    componentAppliedLinearAccClamp=<3000 3000 3000>
    componentLinearAccScale=<0.3 0.3 0.35>
    componentLinearVelScale=<0.2 0.2 0.2>
    overrideWorldGravity {z=-980}
   }
   n24 = RigidBody(ComponentPose=@n23, Alpha=@n22) {
    bClampLinearTranslationLimitToRefPose=true
    bOverrideWorldGravity=true
    overridePhysicsAsset=&/F_MED_Grandeur_Shoelace_r_Physics
    simulationSpace=BaseBoneSpace
    alphaBoolBlend {blendOption=Linear}
    baseBoneRef {boneName=calf_twist_01_r}
    componentAppliedLinearAccClamp=<3000 3000 3000>
    componentLinearAccScale=<0.6 0.6 0.6>
    componentLinearVelScale=<0.3 0.3 0.3>
    overrideWorldGravity {z=-980}
   }
   n25 = RigidBody(ComponentPose=@n24, Alpha=@n22) {
    bClampLinearTranslationLimitToRefPose=true
    bOverrideWorldGravity=true
    overridePhysicsAsset=&/F_MED_Grandeur_Shoelace_l_Physics
    simulationSpace=BaseBoneSpace
    alphaBoolBlend {blendOption=Linear}
    baseBoneRef {boneName=calf_twist_01_l}
    componentAppliedLinearAccClamp=<3000 3000 3000>
    componentLinearAccScale=<0.3 0.3 0.3>
    componentLinearVelScale=<0.2 0.2 0.2>
    overrideWorldGravity {z=-980}
   }
   n26 = "Transform (Modify) Bone - Bone: dyn_hood_MSTR"(ComponentPose=@n25, Alpha=Get Neck Alpha() -> (Neck Alpha:double), Translation="0,-1,-1", Rotation="0,0,-10", Scale="1,1,1") {
    rotationMode=BMM_Additive
    rotationSpace=BCS_BoneSpace
    translationMode=BMM_Additive
    translationSpace=BCS_BoneSpace
    alphaBoolBlend {blendOption=Linear}
    boneToModify {boneName=dyn_hood_MSTR}
    rotation {roll=-10}
    translation {y=-1, z=-1}
   }
   n27 = "Transform (Modify) Bone - Bone: dyn_hood_strap_l_MSTR"(ComponentPose=@n26, Alpha=Get L Arm Alpha() -> (L Arm Alpha:double), Translation="-0.25,1,0", Rotation="0,5,0", Scale="1,1,1") {
    rotationMode=BMM_Additive
    rotationSpace=BCS_BoneSpace
    translationMode=BMM_Additive
    translationSpace=BCS_BoneSpace
    alphaBoolBlend {blendOption=Linear}
    boneToModify {boneName=dyn_hood_strap_l_MSTR}
    rotation {yaw=5}
    translation=<-0.25 1>
   }
   n28 = "Transform (Modify) Bone - Bone: dyn_hood_strap_r_MSTR"(ComponentPose=@n27, Alpha=Get R Arm Alpha() -> (R Arm Alpha:double), Translation="0.25,-1,0", Rotation="0,0,0", Scale="1,1,1") {
    rotationMode=BMM_Additive
    rotationSpace=BCS_BoneSpace
    translationMode=BMM_Additive
    translationSpace=BCS_BoneSpace
    alphaBoolBlend {blendOption=Linear}
    boneToModify {boneName=dyn_hood_strap_r_MSTR}
    translation=<0.25 -1>
   }
   n29 = "Transform (Modify) Bone - Bone: dyn_belt_leg_r_MSTR"!(
    ComponentPose=@n28
    bAlphaBoolEnabled:bool=OR Boolean(A=Get bIsJumping(), B=Get bIsFalling())
    Translation="1,4,3"
    Rotation="0,0,0"
    Scale="1.25,1,1.25"
   ) -> (Pose) {
    alphaInputType=Bool
    rotationMode=BMM_Additive
    rotationSpace=BCS_BoneSpace
    scaleMode=BMM_Additive
    scaleSpace=BCS_BoneSpace
    translationMode=BMM_Additive
    translationSpace=BCS_BoneSpace
    alphaBoolBlend {blendInTime=0.1, blendOption=Linear, blendOutTime=0.1}
    boneToModify {boneName=dyn_belt_leg_r_MSTR}
    scale {x=1.25, z=1.25}
    translation=<1 4 3>
   }
   n30 = "Transform (Modify) Bone - Bone: dyn_belt_leg_r_MSTR"(ComponentPose=@n29, bAlphaBoolEnabled=Get bIsCrouching(), Translation="0.5,3,1", Rotation="0,0,0", Scale="1.25,1,1.25") {
    alphaInputType=Bool
    rotationMode=BMM_Additive
    rotationSpace=BCS_BoneSpace
    scaleMode=BMM_Additive
    scaleSpace=BCS_BoneSpace
    translationMode=BMM_Additive
    translationSpace=BCS_BoneSpace
    alphaBoolBlend {blendInTime=0.1, blendOption=Linear, blendOutTime=0.1}
    boneToModify {boneName=dyn_belt_leg_r_MSTR}
    scale {x=1.25, z=1.25}
    translation=<0.5 3 1>
   }
   n31 = "Transform (Modify) Bone - Bone: dyn_hood_MSTR"!(ComponentPose=@n30, bAlphaBoolEnabled:bool=Get bIsSkydiveFloating(), Translation="0,-4,-4", Rotation="0,0,20", Scale="1,1,1") -> (Pose) {
    alphaInputType=Bool
    rotationMode=BMM_Additive
    rotationSpace=BCS_BoneSpace
    translationMode=BMM_Additive
    translationSpace=BCS_BoneSpace
    alphaBoolBlend {blendInTime=0.1, blendOption=Linear, blendOutTime=0.1}
    boneToModify {boneName=dyn_hood_MSTR}
    rotation {roll=20}
    translation {y=-4, z=-4}
   }
   n32 = Get Belt Crouch Alpha() -> (Belt Crouch Alpha:double)
   n33 = "Transform (Modify) Bone - Bone: dyn_belt_FR_01"(ComponentPose=@n31, Alpha=@n32, Translation="0,0,0", Rotation="0,-145,0", Scale="1,1,1") {
    rotationMode=BMM_Additive
    rotationSpace=BCS_BoneSpace
    translationMode=BMM_Additive
    translationSpace=BCS_BoneSpace
    alphaBoolBlend {blendOption=Linear}
    boneToModify {boneName=dyn_belt_FR_01}
    rotation {yaw=-145}
   }
   n34 = "Transform (Modify) Bone - Bone: dyn_belt_FR_01"(ComponentPose=@n33, Alpha=@n32, Translation="0,-7,0", Rotation="0,0,0", Scale="1,1,1") {
    rotationMode=BMM_Additive
    rotationSpace=BCS_BoneSpace
    translationMode=BMM_Additive
    translationSpace=BCS_BoneSpace
    alphaBoolBlend {blendOption=Linear}
    boneToModify {boneName=dyn_belt_FR_01}
    translation {y=-7}
   }
   n35 = Get Belt Jump Alpha() -> (Belt Jump Alpha:double)
   n36 = "Transform (Modify) Bone - Bone: dyn_belt_FR_01"(ComponentPose=@n34, Alpha=@n35, Translation="0,0,0", Rotation="0,-115,0", Scale="1,1,1") {
    rotationMode=BMM_Additive
    rotationSpace=BCS_BoneSpace
    translationMode=BMM_Additive
    translationSpace=BCS_BoneSpace
    alphaBoolBlend {blendOption=Linear}
    boneToModify {boneName=dyn_belt_FR_01}
    rotation {yaw=-115}
   }
   n37 = "Transform (Modify) Bone - Bone: dyn_belt_FR_01"(ComponentPose=@n36, Alpha=@n35, Translation="0,-4,0", Rotation="0,0,0", Scale="1,1,1") {
    rotationMode=BMM_Additive
    rotationSpace=BCS_BoneSpace
    translationMode=BMM_Additive
    translationSpace=BCS_BoneSpace
    alphaBoolBlend {blendOption=Linear}
    boneToModify {boneName=dyn_belt_FR_01}
    translation {y=-4}
   }
   n38 = Get Run Belt Alpha() -> (Run Belt Alpha:double)
   n39 = "Transform (Modify) Bone - Bone: dyn_belt_FR_01"(ComponentPose=@n37, Alpha=@n38, Translation="0,0,0", Rotation="0,-125,0", Scale="1,1,1") {
    rotationMode=BMM_Additive
    rotationSpace=BCS_BoneSpace
    translationMode=BMM_Additive
    translationSpace=BCS_BoneSpace
    alphaBoolBlend {blendOption=Linear}
    boneToModify {boneName=dyn_belt_FR_01}
    rotation {yaw=-125}
   }
   n40 = "Transform (Modify) Bone - Bone: dyn_belt_FR_01"(ComponentPose=@n39, Alpha=@n38, Translation="0,-5,0", Rotation="0,0,0", Scale="1,1,1") {
    rotationMode=BMM_Additive
    rotationSpace=BCS_BoneSpace
    translationMode=BMM_Additive
    translationSpace=BCS_BoneSpace
    alphaBoolBlend {blendOption=Linear}
    boneToModify {boneName=dyn_belt_FR_01}
    translation {y=-5}
   }
   n41 = "Transform (Modify) Bone - Bone: dyn_belt_FR_01"(ComponentPose=@n40, Alpha=Get Side Run Alpha() -> (Side Run Alpha:double), Translation="0,-1,4", Rotation="0,0,0", Scale="1,1,1") {
    rotationMode=BMM_Additive
    rotationSpace=BCS_BoneSpace
    translationMode=BMM_Additive
    translationSpace=BCS_BoneSpace
    alphaBoolBlend {blendOption=Linear}
    boneToModify {boneName=dyn_belt_FR_01}
    translation {y=-1, z=4}
   }
   n42 = "Transform (Modify) Bone - Bone: dyn_bag_l_MSTR"!(ComponentPose=@n41, bAlphaBoolEnabled:bool=Get bIsCrouching(), Translation="0,-1,0", Rotation="0,0,0", Scale="1,1,1") -> (Pose) {
    alphaInputType=Bool
    rotationMode=BMM_Additive
    rotationSpace=BCS_BoneSpace
    translationMode=BMM_Additive
    translationSpace=BCS_BoneSpace
    alphaBoolBlend {blendInTime=0.1, blendOption=Linear, blendOutTime=0.1}
    boneToModify {boneName=dyn_bag_l_MSTR}
    translation {y=-1}
   }
   n43 = "Transform (Modify) Bone - Bone: backpack_mstr_jnt"!(ComponentPose=@n42, bAlphaBoolEnabled:bool=Get bIsSkyDiving(), Translation="2,0,0", Rotation="0,0,0", Scale="1,1,1") -> (Pose) {
    alphaInputType=Bool
    translationMode=BMM_Replace
    translationSpace=BCS_BoneSpace
    alphaBoolBlend {blendOption=Linear}
    boneToModify {boneName=backpack_mstr_jnt}
    translation {x=2}
   }
   Output Pose(Result=Component To Local(ComponentPose=@n43)) {layerGroup=None}
  ]}
 EventGraph {nodes [
   Event BlueprintUpdateAnimation()
   > n1 = Sequence!(execute) -> (then_0, then_1, then_2:Exec, then_3:Exec)
   n2 = BreakTransform(InTransform=GetDeltaTransformFromRefPose(self=GetOwnerSkeletalMeshComponent(), BoneName=head, BaseName=spine_05))
   > Set Neck Alpha(Neck Alpha:double=MapRangeClamped(Value=BreakRotator(InRot=@n2.Rotation).Yaw, InRangeA=-20, InRangeB=-45)) -> (Output_Get:double)
   n3 = BreakTransform(InTransform=GetDeltaTransformFromRefPose(self=GetOwnerSkeletalMeshComponent(), BoneName=upperArm_l, BaseName=spine_05))
   > Set L Arm Alpha(L Arm Alpha:double=MapRangeClamped(Value=BreakRotator(InRot=@n3.Rotation).Yaw, InRangeA=0, InRangeB=40)) -> (Output_Get:double)
   n4 = BreakTransform(InTransform=GetDeltaTransformFromRefPose(self=GetOwnerSkeletalMeshComponent(), BoneName=upperArm_r, BaseName=spine_05))
   > Set R Arm Alpha(R Arm Alpha:double=MapRangeClamped(Value=BreakRotator(InRot=@n4.Rotation).Yaw, InRangeA=30, InRangeB=60)) -> (Output_Get:double)
   n5 = Branch(execute=@n1.then_1, Condition=Get bIsCrouching())
   n6 = BreakTransform(InTransform=GetDeltaTransformFromRefPose(self=GetOwnerSkeletalMeshComponent(), BoneName=thigh_r, BaseName=pelvis))
   > Set Belt Crouch Alpha(Belt Crouch Alpha:double=MapRangeClamped(Value=BreakRotator(InRot=@n6.Rotation).Yaw, InRangeA=-25, InRangeB=-100)) -> (Output_Get:double)
   Set Belt Crouch Alpha(execute=@n5.else, Belt Crouch Alpha=0)
   n7 = Branch(execute=@n1.then_2, Condition=OR Boolean(A=Get bIsJumping(), B=Get bIsFalling()))
   n8 = BreakTransform(InTransform=GetDeltaTransformFromRefPose(self=GetOwnerSkeletalMeshComponent(), BoneName=thigh_r, BaseName=pelvis))
   > Set Belt Jump Alpha(Belt Jump Alpha:double=MapRangeClamped(Value=BreakRotator(InRot=@n8.Rotation).Yaw, InRangeA=-15, InRangeB=-80)) -> (Output_Get:double)
   Set Belt Jump Alpha(execute=@n7.else, Belt Jump Alpha=0)
   n9 = Get bIsCrouching()
   n10 = Get bIsJumping()
   n11 = Get bIsFalling()
   n12 = Branch(execute=@n1.then_3, Condition=NOT Boolean(A=OR Boolean!(A=@n10, B=@n11, C:bool=@n9) -> (ReturnValue)))
   n13 = BreakTransform(InTransform=GetDeltaTransformFromRefPose(self=GetOwnerSkeletalMeshComponent(), BoneName=thigh_r, BaseName=pelvis))
   > Set Run Belt Alpha(Run Belt Alpha:double=MapRangeClamped(Value=BreakRotator(InRot=@n13.Rotation).Yaw, InRangeA=0, InRangeB=-70)) -> (Output_Get:double)
   n14 = BreakTransform(InTransform=GetDeltaTransformFromRefPose(self=GetOwnerSkeletalMeshComponent(), BoneName=thigh_r, BaseName=pelvis))
   > Set Side Run Alpha(Side Run Alpha:double=MapRangeClamped(Value=BreakRotator(InRot=@n14.Rotation).Pitch, InRangeA=0, InRangeB=20)) -> (Output_Get:double)
   Set Run Belt Alpha(execute=@n12.else, Run Belt Alpha=0)
   > Set Side Run Alpha(Side Run Alpha=0)
  ]}
]
variables [
 Belt Crouch Alpha {type=number}
 Belt Jump Alpha {type=number}
 L Arm Alpha {type=number}
 Neck Alpha {type=number}
 R Arm Alpha {type=number}
 Run Belt Alpha {type=number}
 Side Run Alpha {type=number}
]
