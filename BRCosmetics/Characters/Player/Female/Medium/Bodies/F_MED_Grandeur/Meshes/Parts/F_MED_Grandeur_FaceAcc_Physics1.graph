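# Physics asset for the face accessory mesh of the Grandeur female medium player cosmetic, simulating four dyn_hair chains: left and right, front and back.
# Each chain runs from a master bone on the head through three box segments, with head and spine_05 capsules forming the body shapes.
# Angular limits shape each chain, while master-to-head and back-root joints swing freely.
# Free-angle cross constraints tie each side's back outer segments to the matching front ones.
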PhysicsAsset /BRCosmetics/Characters/Player/Female/Medium/Bodies/F_MED_Grandeur/Meshes/Parts/F_MED_Grandeur_FaceAcc_Physics1
bodies {
 dyn_hair_LB_01 [{shapeType=Box, extentX=1.01, extentY=1.01, extentZ=1.01, shapeName=dyn_hair_LB_01_box, rotation=<rot 0 -90 0>}]
 dyn_hair_LB_02 [{shapeType=Box, extentX=5, extentY=1, extentZ=5, shapeName=dyn_hair_LB_02_box, center=<4.5 0 0>}]
 dyn_hair_LB_03 [{shapeType=Box, extentX=5, extentY=1, extentZ=5, shapeName=dyn_hair_LB_03_box, center=<4.5 0 0>}]
 dyn_hair_LB_MSTR [{shapeType=Capsule, length=0.505, radius=0.505, shapeName=dyn_hair_LB_MSTR_capsule, rotation=<rot 0 -90 0>}]
 dyn_hair_LF_01 [{shapeType=Box, extentX=5, extentY=5, extentZ=1.01, shapeName=dyn_hair_RF_01_box, center=<4 0 0>, rotation=<rot 0 -90 0>}]
 dyn_hair_LF_02 [{shapeType=Box, extentX=5, extentY=5, extentZ=1, shapeName=dyn_hair_LF_02_box, center=<4 0 0>}]
 dyn_hair_LF_03 [{shapeType=Box, extentX=5, extentY=5, extentZ=1, shapeName=dyn_hair_LF_03_box, center=<4 0 0>}]
 dyn_hair_LF_MSTR [{shapeType=Capsule, length=0.505, radius=0.505, shapeName=dyn_hair_LF_MSTR_capsule, rotation=<rot 0 -90 0>}]
 dyn_hair_RB_01 [{shapeType=Box, extentX=1.01, extentY=1.01, extentZ=1.01, shapeName=dyn_hair_RB_01_box, rotation=<rot 0 -90 0>}]
 dyn_hair_RB_02 [{shapeType=Box, extentX=5, extentY=1, extentZ=5, shapeName=dyn_hair_RB_02_box, center=<4.5 0 0>}]
 dyn_hair_RB_03 [{shapeType=Box, extentX=5, extentY=1, extentZ=5, shapeName=dyn_hair_RB_03_box, center=<4.5 0 0>}]
 dyn_hair_RB_MSTR [{shapeType=Capsule, length=0.505, radius=0.505, shapeName=dyn_hair_RB_MSTR_capsule, rotation=<rot 0 -90 0>}]
 dyn_hair_RF_01 [{shapeType=Box, extentX=5, extentY=5, extentZ=1.01, shapeName=dyn_hair_RF_01_box, center=<4 0 0>, rotation=<rot 0 -90 0>}]
 dyn_hair_RF_02 [{shapeType=Box, extentX=5, extentY=5, extentZ=1, shapeName=dyn_hair_RF_02_box, center=<4 0 0>}]
 dyn_hair_RF_03 [{shapeType=Box, extentX=5, extentY=5, extentZ=1, shapeName=dyn_hair_RF_03_box, center=<4 0 0>}]
 dyn_hair_RF_MSTR [{shapeType=Capsule, length=0.505, radius=0.505, shapeName=dyn_hair_RF_MSTR_capsule, rotation=<rot 0 -90 0>}]
 head [{shapeType=Capsule, length=7.2355, radius=7.6128, center=<2.4973 1.7216 0.4704>, rotation=<rot 4.4832 99.2356 -89.9999>}]
 spine_05 [
  {shapeType=Capsule, length=18.8608, radius=7.2194, shapeName=spine_05_capsule, center=<13.1229 -13.1082 0>}
  {shapeType=Capsule, length=18.3608, radius=4.5, shapeName=spine_05_capsule, center=<13.1229 -5.8557 -14.1136>, rotation=<rot 0 0 -50>}
  {shapeType=Capsule, length=18.3608, radius=4.5, shapeName=spine_05_capsule, center=<13.1229 -5.8557 14.1136>, rotation=<rot 0 0 50>}
 ]
}
constraints [
 {
  bone1Name=dyn_hair_LF_MSTR
  bone2Name=head
  swing1LimitDegrees=45
  swing1Motion=Free
  swing2LimitDegrees=45
  swing2Motion=Free
  twistLimitDegrees=45
  twistMotion=Free
 }
 {bone1Name=dyn_hair_LF_01, bone2Name=dyn_hair_LF_MSTR, swing1LimitDegrees=4, swing2LimitDegrees=4, twistLimitDegrees=1}
 {bone1Name=dyn_hair_LF_02, bone2Name=dyn_hair_LF_01, swing1LimitDegrees=45, swing2LimitDegrees=20, twistLimitDegrees=2}
 {bone1Name=dyn_hair_LF_03, bone2Name=dyn_hair_LF_02, swing1LimitDegrees=45, swing2LimitDegrees=20, twistLimitDegrees=2}
 {
  bone1Name=dyn_hair_RF_MSTR
  bone2Name=head
  swing1LimitDegrees=45
  swing1Motion=Free
  swing2LimitDegrees=45
  swing2Motion=Free
  twistLimitDegrees=45
  twistMotion=Free
 }
 {
  bone1Name=dyn_hair_LB_MSTR
  bone2Name=head
  swing1LimitDegrees=45
  swing1Motion=Free
  swing2LimitDegrees=45
  swing2Motion=Free
  twistLimitDegrees=45
  twistMotion=Free
 }
 {
  bone1Name=dyn_hair_RB_MSTR
  bone2Name=head
  swing1LimitDegrees=45
  swing1Motion=Free
  swing2LimitDegrees=45
  swing2Motion=Free
  twistLimitDegrees=45
  twistMotion=Free
 }
 {bone1Name=dyn_hair_RF_01, bone2Name=dyn_hair_RF_MSTR, swing1LimitDegrees=4, swing2LimitDegrees=4, twistLimitDegrees=1}
 {bone1Name=dyn_hair_RF_02, bone2Name=dyn_hair_RF_01, swing1LimitDegrees=45, swing2LimitDegrees=20, twistLimitDegrees=2}
 {bone1Name=dyn_hair_RF_03, bone2Name=dyn_hair_RF_02, swing1LimitDegrees=45, swing2LimitDegrees=20, twistLimitDegrees=2}
 {
  bone1Name=dyn_hair_LB_01
  bone2Name=dyn_hair_LB_MSTR
  swing1LimitDegrees=45
  swing1Motion=Free
  swing2LimitDegrees=45
  swing2Motion=Free
  twistLimitDegrees=45
  twistMotion=Free
 }
 {bone1Name=dyn_hair_LB_02, bone2Name=dyn_hair_LB_01, swing1LimitDegrees=30, swing2LimitDegrees=20, twistLimitDegrees=2}
 {bone1Name=dyn_hair_LB_03, bone2Name=dyn_hair_LB_02, swing1LimitDegrees=30, swing2LimitDegrees=20, twistLimitDegrees=2}
 {
  bone1Name=dyn_hair_RB_01
  bone2Name=dyn_hair_RB_MSTR
  swing1LimitDegrees=45
  swing1Motion=Free
  swing2LimitDegrees=45
  swing2Motion=Free
  twistLimitDegrees=45
  twistMotion=Free
 }
 {bone1Name=dyn_hair_RB_02, bone2Name=dyn_hair_RB_01, swing1LimitDegrees=30, swing2LimitDegrees=20, twistLimitDegrees=2}
 {bone1Name=dyn_hair_RB_03, bone2Name=dyn_hair_RB_02, swing1LimitDegrees=30, swing2LimitDegrees=20, twistLimitDegrees=2}
 {
  bone1Name=dyn_hair_RB_03
  bone2Name=dyn_hair_RF_03
  swing1LimitDegrees=45
  swing1Motion=Free
  swing2LimitDegrees=45
  swing2Motion=Free
  twistLimitDegrees=45
  twistMotion=Free
 }
 {
  bone1Name=dyn_hair_RB_02
  bone2Name=dyn_hair_RF_02
  swing1LimitDegrees=45
  swing1Motion=Free
  swing2LimitDegrees=45
  swing2Motion=Free
  twistLimitDegrees=45
  twistMotion=Free
 }
 {
  bone1Name=dyn_hair_LB_03
  bone2Name=dyn_hair_LF_03
  swing1LimitDegrees=45
  swing1Motion=Free
  swing2LimitDegrees=45
  swing2Motion=Free
  twistLimitDegrees=45
  twistMotion=Free
 }
 {
  bone1Name=dyn_hair_LB_02
  bone2Name=dyn_hair_LF_02
  swing1LimitDegrees=45
  swing1Motion=Free
  swing2LimitDegrees=45
  swing2Motion=Free
  twistLimitDegrees=45
  twistMotion=Free
 }
 {
  bone1Name=head
  bone2Name=spine_05
  swing1LimitDegrees=45
  swing1Motion=Free
  swing2LimitDegrees=45
  swing2Motion=Free
  twistLimitDegrees=45
  twistMotion=Free
 }
]
