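# Post-copy AnimBP for the F_MED_Grandeur face accessory cosmetic part, adding secondary physics to the pose copied from the attached parent mesh.
# A RigidBody node simulates in neck_01 base-bone space, with damped component motion and gravity overridden each update by a deprecated CharacterPartAnimInstance helper.
# The update also lowers simulation alpha while skydiving.
# During emotes, a blended additive offset slightly shifts C_jaw after the simulation.
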AnimBlueprint /BRCosmetics/Characters/Player/Female/Medium/Bodies/F_MED_Grandeur/Meshes/Parts/F_MED_Grandeur_FaceAcc_AnimBP
target_skeleton=&/F_MED_Grandeur_FaceAcc_Skeleton
graphs [
 AnimGraph {nodes [
   n1 = RigidBody!(
    ComponentPose=Local To Component(LocalPose=Copy Pose From Mesh() {bUseAttachedParent=true})
    Alpha=Get Skydive alpha() -> (Skydive alpha:double)
    OverrideWorldGravity:Vector=Get Gravity Override()
    ExternalForce
   ) -> (Pose) {
    bClampLinearTranslationLimitToRefPose=true
    bOverrideWorldGravity=true
    overridePhysicsAsset=&/F_MED_Grandeur_FaceAcc_Physics1
    simulationSpace=BaseBoneSpace
    alphaBoolBlend {blendOption=Linear}
    baseBoneRef {boneName=neck_01}
    componentAppliedLinearAccClamp=<3000 3000 3000>
    componentLinearAccScale=<0.3 0.3 0.45>
    componentLinearVelScale=<0.1 0.1 0.1>
    overrideWorldGravity {z=-980}
   }
   n2 = "Transform (Modify) Bone - Bone: C_jaw"!(ComponentPose=@n1, bAlphaBoolEnabled:bool=Get bIsPlayingEmote(), Translation="0,0.2,0", Rotation="0,0,0", Scale="1,1,1") -> (Pose) {
    alphaInputType=Bool
    translationMode=BMM_Additive
    translationSpace=BCS_BoneSpace
    alphaBoolBlend {blendInTime=0.1, blendOption=Linear, blendOutTime=0.1}
    boneToModify {boneName=C_jaw}
    translation {y=0.2}
   }
   Output Pose(Result=Component To Local(ComponentPose=@n2)) {layerGroup=None}
  ]}
 EventGraph {nodes [
   Event BlueprintUpdateAnimation()
   > n1 = Sequence()
   > n2 = Gravity Override State Function DEPRECATED(CharacterPartAnimInstance=Self-Reference(), GameTimeInSeconds=GetGameTimeInSeconds())
   > Set Gravity Override(Gravity Override=@n2)
   n3 = Branch(execute=@n1.then_1, Condition=Get bIsSkyDiving())
   > Set Skydive alpha(Skydive alpha:double=0.6) -> (Output_Get:double)
   Set Skydive alpha(execute=@n3.else, Skydive alpha=1)
  ]}
]
variables [
 Gravity Override {type=Vector}
 Skydive alpha {type=number}
]
